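# Translucent unlit Niagara trail material for the idle effect of the ScareyBeary pickaxe.
# SplineThicken sets strand width, while time-driven sine waves and a clamped pull toward a dynamic-parameter position (X fed to all axes) add world position offset.
# Emissive combines two color-curve atlas rows, a panning trail texture and particle color.
# Opacity blends two panning textures, masked by a sphere around the particle, with depth fade and particle alpha.
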Material /BRCosmetics/Weapons/FORT_Melee/Pickaxe_ScareyBeary/FX/Material/M_Pickaxe_ScareyBeary_IdleTrail
graph {
 properties {
  bUsedWithNiagaraMeshParticles=true
  bUsedWithNiagaraRibbons=true
  blendMode=BLEND_Translucent
  pixelDepthOffsetMode=PDOM_Legacy
  shadingModel=MSM_Unlit
  twoSided=true
  usedShadingModels=Unlit
 }
 nodes [
  n1 = DynamicParameter() {defaultValue=<rgba 0 10 0 0>, paramNames [WidthBase, WidthTip, WPO Offset, Use Default UV WPO]}
  n2 = Multiply(
   A=ScalarParameter() {defaultValue=1, parameterName=SplineWidth}
   B=SphereMask(A=TextureCoordinate() {vTiling=0}, B=Constant2Vector() {r=0.5}) {attenuationRadius=0.5, hardnessPercent=50}
  )
  n3 = Multiply(A=@n1.WidthBase, B=@n2)
  n4 = Multiply(A=@n1.WidthTip, B=@n2)
  n5 = StaticBool()
  n6 = TextureCoordinate() {coordinateIndex=1}
  n7 = TextureCoordinate()
  n8 = DynamicParameter() {defaultValue=<rgba 0 10 0 0>, paramNames [WidthBase, WidthTip, WPO Offset, Use Default UV WPO]}
  n9 = Add(
   A=ComponentMask(None=LinearInterpolate(A=@n6, B=@n7, Alpha=@n8.Use Default UV WPO)) {r=true}
   B=DynamicParameter() {defaultValue=<rgba 0 10 0 0>, paramNames [WidthBase, WidthTip, WPO Offset, Use Default UV WPO]}.WPO Offset
  )
  n10 = ScalarParameter() {defaultValue=-0.3, parameterName=WPO_Speed}
  n11 = Multiply(A=Time(), B=@n10) {constB=-0.3}
  n12 = Multiply(A=Constant3Vector() {constant=<rgba 1 0 0 0>}, B=Sine(None=Add(A=@n9, B=@n11)) {period=0.5})
  n13 = TextureCoordinate() {coordinateIndex=1}
  n14 = TextureCoordinate()
  n15 = DynamicParameter() {defaultValue=<rgba 0 10 0 0>, paramNames [WidthBase, WidthTip, WPO Offset, Use Default UV WPO]}
  n16 = Add(
   A=ComponentMask(None=LinearInterpolate(A=@n13, B=@n14, Alpha=@n15.Use Default UV WPO)) {r=true}
   B=DynamicParameter() {defaultValue=<rgba 0 10 0 0>, paramNames [WidthBase, WidthTip, WPO Offset, Use Default UV WPO]}.WPO Offset
  )
  n17 = Multiply(A=Time(), B=@n10) {constB=-0.3}
  n18 = Multiply(A=Constant3Vector() {constant=<rgba 0 1 0 0>}, B=Sine(None=Add(A=@n16, B=@n17)) {period=0.8})
  n19 = TextureCoordinate() {coordinateIndex=1}
  n20 = TextureCoordinate()
  n21 = DynamicParameter() {defaultValue=<rgba 0 10 0 0>, paramNames [WidthBase, WidthTip, WPO Offset, Use Default UV WPO]}
  n22 = Power(Base=ComponentMask(None=LinearInterpolate(A=@n19, B=@n20, Alpha=@n21.Use Default UV WPO)) {r=true})
  n23 = Multiply(
   A=Multiply(A=Add(A=@n12, B=@n18), B=ScalarParameter() {defaultValue=0.5, parameterName=WPO Intensity})
   B=Saturate(None=Multiply(A=@n22) {constB=3})
  )
  n24 = TextureCoordinate()
  n25 = DynamicParameter() {parameterIndex=1, defaultValue=<rgba 0 10 0 1>, paramNames [X,Y,Z,WPOIntensity]}
  n26 = Multiply(A=@n25.WPOIntensity, B=Power(Base=ComponentMask(None=@n24) {r=true}, Exp=ScalarParameter() {defaultValue=2, parameterName=WPOBlendPower}))
  n27 = Clamp(
   None=Subtract(
    A=MaterialFunctionCall(X=@n25.X, Y=@n25.X, Z=@n25.X) {materialFunction=&MakeFloat3}.Result
    B=ParticlePositionWS()
   )
   Min=ScalarParameter() {defaultValue=-30, parameterName=Clamp Min}
   Max=ScalarParameter() {defaultValue=30, parameterName=Clamp Max}
  )
  Add(
   A=Add(
    A=MaterialFunctionCall(WidthBase=@n3, WidthTip=@n4, Expand U or V UV Channel=@n5) {materialFunction=&SplineThicken}.WorldPositionOffset
    B=Transform(None=@n23) {transformSourceType=TRANSFORMSOURCE_Local}
   )
   B=LinearInterpolate(A=Constant(), B=LinearInterpolate(B=@n27, Alpha=@n26), Alpha=ScalarParameter() {defaultValue=1, parameterName=Use Interpolate WPO})
  )
  n28 = ParticleColor()
  n29 = CurveAtlasRowParameter(CurveTime=ComponentMask(None=TextureCoordinate()) {r=true}) {
   atlas=/Game/Weapons/FORT_Melee/Pickaxe_ScareyBeary/FX/CA_Pickaxe_ScareyBeary
   curve=/Game/Weapons/FORT_Melee/Pickaxe_ScareyBeary/FX/CV_Pickaxe_ScareyBeary
   parameterName=Color Curve
  }
  n30 = CurveAtlasRowParameter(CurveTime=ComponentMask(None=TextureCoordinate()) {g=true}) {
   atlas=&CA_Pickaxe_ScareyBeary
   curve=/Game/Weapons/FORT_Melee/Pickaxe_ScareyBeary/FX/CV_PIckaxe_ScareyBeary02
   defaultValue=1
   parameterName=Color Curve02
  }
  n31 = AppendVector(
   A=DynamicParameter() {defaultValue=<rgba 0 10 0 0>, paramNames [WidthBase, WidthTip, WPO Offset, Use Default UV WPO]}.WPO Offset
   B=Constant()
  )
  n32 = Panner(Coordinate=Add(A=TextureCoordinate(), B=@n31)) {speedX=-0.5}
  n33 = Power(Base=Multiply(A=TextureSampleParameter2D(UVs=@n32) {parameterName=BaseTexture, texture=/BRCosmetics/Weapons/FORT_Melee/Pickaxe_Gimmick_Female/FX/T_Pickaxe_Gimmick_Female_Trail_03}.R)) {constExponent=1}
  n34 = Add(A=@n29, B=Desaturation(None=Multiply(A=@n30, B=@n33), Fraction=ScalarParameter() {defaultValue=-1, parameterName=Desaturation})) {constB=0}
  MaterialFunctionCall(Emissive=Desaturation(None=Multiply(A=@n34, B=@n28.RGB), Fraction=ScalarParameter() {parameterName=All_Desaturation})) {materialFunction=&MF_FX_EmissiveAdjustments}
  n35 = Add(A=@n31, B=TextureCoordinate())
  n36 = Add(
   A=TextureSampleParameter2D(UVs=Panner(Coordinate=@n35) {speedX=-0.5}) {parameterName=LerpTexture, texture=/BRCosmetics/Weapons/FORT_Melee/Pickaxe_Inferno/FX/T_Pickaxe_Inferno_Trail_01}.R
   B=Multiply(A=@n33) {constB=0.3}
  ) {constB=0}
  n37 = SphereMask(A=WorldPosition().XYZ, B=ParticlePositionWS(), Radius=ScalarParameter() {defaultValue=50, parameterName=Radius}) {attenuationRadius=50, hardnessPercent=30}
  Multiply(A=@n28.A, B=DepthFade(Opacity=Multiply(A=@n36, B=@n37)) {fadeDistanceDefault=5})
  ScalarParameter() {parameterName=Use Default UV WPO}
  ScalarParameter() {parameterName=Use Default UV WPO}
  ScalarParameter() {parameterName=Use Default UV WPO}
 ]
}
