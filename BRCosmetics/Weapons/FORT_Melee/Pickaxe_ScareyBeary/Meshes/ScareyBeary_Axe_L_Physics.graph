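# Physics asset for the left ScareyBeary pickaxe mesh, defining bodies and joints for the motion of its fur.
# Eight base fur bones carry small capsules, each paired with a larger, offset capsule on a matching dyn fur bone.
# Each pair has its own constraint with narrow swing and near-zero twist, with no links between neighboring tufts.
# The first and last pairs have the tightest swing limits.
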PhysicsAsset /BRCosmetics/Weapons/FORT_Melee/Pickaxe_ScareyBeary/Meshes/ScareyBeary_Axe_L_Physics
bodies {
 base_fur_1_l [{shapeType=Capsule, length=0.505, radius=0.505, shapeName=base_fur_1_l_capsule, rotation=<rot 0 -90 0>}]
 base_fur_2_l [{shapeType=Capsule, length=0.505, radius=0.505, shapeName=base_fur_2_l_capsule, rotation=<rot 0 -90 0>}]
 base_fur_3_l [{shapeType=Capsule, length=0.505, radius=0.505, shapeName=base_fur_3_l_capsule, rotation=<rot 0 -90 0>}]
 base_fur_4_l [{shapeType=Capsule, length=0.505, radius=0.505, shapeName=base_fur_4_l_capsule, rotation=<rot 0 -90 0>}]
 base_fur_5_l [{shapeType=Capsule, length=0.505, radius=0.505, shapeName=base_fur_5_l_capsule, rotation=<rot 0 -90 0>}]
 base_fur_6_l [{shapeType=Capsule, length=0.505, radius=0.505, shapeName=base_fur_6_l_capsule, rotation=<rot 0 -90 0>}]
 base_fur_7_l [{shapeType=Capsule, length=0.505, radius=0.505, shapeName=base_fur_7_l_capsule, rotation=<rot 0 -90 0>}]
 base_fur_8_l [{shapeType=Capsule, length=0.505, radius=0.505, shapeName=base_fur_8_l_capsule, rotation=<rot 0 -90 0>}]
 dyn_fur_1_l [{shapeType=Capsule, length=1, radius=1, shapeName=dyn_fur_1_l_capsule, center=<1.1716 0 0>, rotation=<rot 90 0 0>}]
 dyn_fur_2_l [{shapeType=Capsule, length=1, radius=1, shapeName=dyn_fur_2_l_capsule, center=<1.4667 0 0>, rotation=<rot 90 0 0>}]
 dyn_fur_3_l [{shapeType=Capsule, length=1, radius=1, shapeName=dyn_fur_3_l_capsule, center=<1.6069 0 0>, rotation=<rot 90 0 0>}]
 dyn_fur_4_l [{shapeType=Capsule, length=1, radius=1, shapeName=dyn_fur_4_l_capsule, center=<1.3736 0 0>, rotation=<rot 90 0 0>}]
 dyn_fur_5_l [{shapeType=Capsule, length=1, radius=1, shapeName=dyn_fur_5_l_capsule, center=<1.3736 0 0>, rotation=<rot 90 0 0>}]
 dyn_fur_6_l [{shapeType=Capsule, length=1, radius=1, shapeName=dyn_fur_6_l_capsule, center=<1.3736 0 0>, rotation=<rot 90 0 0>}]
 dyn_fur_7_l [{shapeType=Capsule, length=1, radius=1, shapeName=dyn_fur_7_l_capsule, center=<1.3736 0 0>, rotation=<rot 90 0 0>}]
 dyn_fur_8_l [{shapeType=Capsule, length=1, radius=1, shapeName=dyn_fur_8_l_capsule, center=<0.9382 0 0>, rotation=<rot 90 0 0>}]
}
constraints [
 {bone1Name=dyn_fur_1_l, bone2Name=base_fur_1_l, swing1LimitDegrees=5, swing2LimitDegrees=2.5, twistLimitDegrees=0}
 {bone1Name=dyn_fur_2_l, bone2Name=base_fur_2_l, swing1LimitDegrees=15, swing2LimitDegrees=10, twistLimitDegrees=0}
 {bone1Name=dyn_fur_3_l, bone2Name=base_fur_3_l, swing1LimitDegrees=10, swing2LimitDegrees=10, twistLimitDegrees=0}
 {bone1Name=dyn_fur_4_l, bone2Name=base_fur_4_l, swing1LimitDegrees=15, swing2LimitDegrees=10, twistLimitDegrees=0}
 {bone1Name=dyn_fur_5_l, bone2Name=base_fur_5_l, swing1LimitDegrees=15, swing2LimitDegrees=10, twistLimitDegrees=0}
 {bone1Name=dyn_fur_6_l, bone2Name=base_fur_6_l, swing1LimitDegrees=15, swing2LimitDegrees=10, twistLimitDegrees=0}
 {bone1Name=dyn_fur_8_l, bone2Name=base_fur_8_l, swing1LimitDegrees=5, swing2LimitDegrees=2.5, twistLimitDegrees=1}
 {bone1Name=dyn_fur_7_l, bone2Name=base_fur_7_l, swing1LimitDegrees=15, swing2LimitDegrees=10, twistLimitDegrees=1}
]
