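# Physics asset for the right-hand ScareyBeary pickaxe cosmetic, setting up eight fur strands for physical secondary motion.
# Each strand pairs a small capsule on a base fur bone with a larger, offset capsule on a dynamic fur bone.
# A constraint joins each dynamic bone to its base, allowing narrow, asymmetric swing and little or no twist.
# Strands are not constrained to one another, and only fur bones have bodies.
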PhysicsAsset /BRCosmetics/Weapons/FORT_Melee/Pickaxe_ScareyBeary/Meshes/ScareyBeary_Axe_R_Physics
bodies {
 base_fur_1_r [{shapeType=Capsule, length=0.505, radius=0.505, shapeName=base_fur_1_r_capsule, rotation=<rot 0 -90 0>}]
 base_fur_2_r [{shapeType=Capsule, length=0.505, radius=0.505, shapeName=base_fur_2_r_capsule, rotation=<rot 0 -90 0>}]
 base_fur_3_r [{shapeType=Capsule, length=0.505, radius=0.505, shapeName=base_fur_3_r_capsule, rotation=<rot 0 -90 0>}]
 base_fur_4_r [{shapeType=Capsule, length=0.505, radius=0.505, shapeName=base_fur_4_r_capsule, rotation=<rot 0 -90 0>}]
 base_fur_5_r [{shapeType=Capsule, length=0.505, radius=0.505, shapeName=base_fur_5_r_capsule, rotation=<rot 0 -90 0>}]
 base_fur_6_r [{shapeType=Capsule, length=0.505, radius=0.505, shapeName=base_fur_6_r_capsule, rotation=<rot 0 -90 0>}]
 base_fur_7_r [{shapeType=Capsule, length=0.505, radius=0.505, shapeName=base_fur_7_r_capsule, rotation=<rot 0 -90 0>}]
 base_fur_8_r [{shapeType=Capsule, length=0.505, radius=0.505, shapeName=base_fur_8_r_capsule, rotation=<rot 0 -90 0>}]
 dyn_fur_1_r [{shapeType=Capsule, length=1, radius=1, shapeName=dyn_fur_1_r_capsule, center=<2 0 0>, rotation=<rot 90 0 0>}]
 dyn_fur_2_r [{shapeType=Capsule, length=1, radius=1, shapeName=dyn_fur_2_r_capsule, center=<2 0 0>, rotation=<rot 90 0 0>}]
 dyn_fur_3_r [{shapeType=Capsule, length=1, radius=1, shapeName=dyn_fur_3_r_capsule, center=<2 0 0>, rotation=<rot 90 0 0>}]
 dyn_fur_4_r [{shapeType=Capsule, length=1, radius=1, shapeName=dyn_fur_4_r_capsule, center=<2 0 0>, rotation=<rot 90 0 0>}]
 dyn_fur_5_r [{shapeType=Capsule, length=1, radius=1, shapeName=dyn_fur_5_r_capsule, center=<2 0 0>, rotation=<rot 90 0 0>}]
 dyn_fur_6_r [{shapeType=Capsule, length=1, radius=1, shapeName=dyn_fur_6_r_capsule, center=<2 0 0>, rotation=<rot 90 0 0>}]
 dyn_fur_7_r [{shapeType=Capsule, length=1, radius=1, shapeName=dyn_fur_7_r_capsule, center=<2 0 0>, rotation=<rot 90 0 0>}]
 dyn_fur_8_r [{shapeType=Capsule, length=1, radius=1, shapeName=dyn_fur_8_r_capsule, center=<2 0 0>, rotation=<rot 90 0 0>}]
}
constraints [
 {bone1Name=dyn_fur_1_r, bone2Name=base_fur_1_r, swing1LimitDegrees=15, swing2LimitDegrees=10, twistLimitDegrees=0}
 {bone1Name=dyn_fur_2_r, bone2Name=base_fur_2_r, swing1LimitDegrees=15, swing2LimitDegrees=10, twistLimitDegrees=0}
 {bone1Name=dyn_fur_3_r, bone2Name=base_fur_3_r, swing1LimitDegrees=15, swing2LimitDegrees=10, twistLimitDegrees=0}
 {bone1Name=dyn_fur_4_r, bone2Name=base_fur_4_r, swing1LimitDegrees=15, swing2LimitDegrees=10, twistLimitDegrees=0}
 {bone1Name=dyn_fur_5_r, bone2Name=base_fur_5_r, swing1LimitDegrees=15, swing2LimitDegrees=10, twistLimitDegrees=0}
 {bone1Name=dyn_fur_6_r, bone2Name=base_fur_6_r, swing1LimitDegrees=15, swing2LimitDegrees=10, twistLimitDegrees=0}
 {bone1Name=dyn_fur_8_r, bone2Name=base_fur_8_r, swing1LimitDegrees=15, swing2LimitDegrees=10, twistLimitDegrees=1}
 {bone1Name=dyn_fur_7_r, bone2Name=base_fur_7_r, swing1LimitDegrees=15, swing2LimitDegrees=10, twistLimitDegrees=1}
]
